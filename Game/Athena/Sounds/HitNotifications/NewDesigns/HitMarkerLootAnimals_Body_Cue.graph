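# Sound cue for the body-hit marker on loot animals, layering the standard body hit sound with a chime.
# A mixer combines one of three random body hit waves with one of six random chimes, which is slightly delayed and mixed louder.
# Each layer has its own modulator for variation.
# Hit-notification attenuation, sound class and a two-voice, stop-oldest concurrency set govern playback.
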SoundCue /Game/Athena/Sounds/HitNotifications/NewDesigns/HitMarkerLootAnimals_Body_Cue
root=Mixer_0
nodes [
 {id=Delay_0, children [Random_1], properties {delayMax=0.05, delayMin=0.03}}
 {id=Mixer_0, children [Modulator_0,Modulator_1], properties {inputVolume [0.8,2]}}
 {id=Modulator_0, children [Random_0]}
 {id=Modulator_1, children [Delay_0], properties {pitchMax=1.1}}
 {id=Random_0, children [WavePlayer_0,WavePlayer_1,WavePlayer_2], properties {bShouldExcludeFromBranchCulling=true, weights [1,1,1]}}
 {
  id=Random_1
  children [WavePlayer_3,WavePlayer_4,WavePlayer_5,WavePlayer_6,WavePlayer_7,WavePlayer_8]
  properties {bShouldExcludeFromBranchCulling=true, weights [1,1,1,1,1,1]}
 }
 {id=WavePlayer_0, properties {soundWaveAssetPtr=/Game/Athena/Sounds/HitNotifications/HitMarker_Body_01}}
 {id=WavePlayer_1, properties {soundWaveAssetPtr=/Game/Athena/Sounds/HitNotifications/HitMarker_Body_02}}
 {id=WavePlayer_2, properties {soundWaveAssetPtr=/Game/Athena/Sounds/HitNotifications/HitMarker_Body_03}}
 {id=WavePlayer_3, properties {soundWaveAssetPtr=&/LootAnimalsHitChimes_01}}
 {id=WavePlayer_4, properties {soundWaveAssetPtr=&/LootAnimalsHitChimes_02}}
 {id=WavePlayer_5, properties {soundWaveAssetPtr=&/LootAnimalsHitChimes_03}}
 {id=WavePlayer_6, properties {soundWaveAssetPtr=&/LootAnimalsHitChimes_04}}
 {id=WavePlayer_7, properties {soundWaveAssetPtr=&/LootAnimalsHitChimes_05}}
 {id=WavePlayer_8, properties {soundWaveAssetPtr=&/LootAnimalsHitChimes_06}}
]
properties {
 attenuationSettings=/Game/Athena/Sounds/HitNotifications/Fort_HitMarker_Instigate_Att
 maxDistance=2097152
 soundClassObject=&Fort_2DHitNotifications_Instigate_PSM
 volumeMultiplier=0.7
 concurrencyOverrides {maxCount=2, resolutionRule=StopOldest}
 concurrencySet [&HitNotifications_TwoMaximum_StopOldest_VolumeScale]
}
